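; Interactable "Time Attack" actor that grants the ring time attack quest through cheat console commands.
; On interact, it gives the interacting player's controller the IcicleLarva special gadget, waits briefly, then runs the grant quest command.
; A separate GrantRingsQuest event, commented as test-only, runs the same quest command for every player pawn's controller.
; Interaction is always allowed; begin play, overlap and tick are empty.
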(event EventBeginPlay)

(event Collision|EventActorBeginOverlap
  (params (OtherActor :type Actor%20Object%20Reference)))

(event EventTick
  (params (DeltaSeconds :type Float%20%28single-precision%29)))

(event Custom|GrantRingsQuest
  (bind _outfortplayerpawns (Fortnite|GetFortPlayerPawns))
  (for _array_element _outfortplayerpawns
    ; TestOnly
    (Development|ExecuteConsoleCommand "cheat grantquest Quest_PistolCanary_Found_RingTimeAttack_Urgent_Repeat" (Fortnite|GetFortPlayerControllerfromActor _array_element))))

(event Interaction|EventBlueprintOnInteract
  (params (InteractingPawn :type Fort%20Pawn%20Object%20Reference) (InteractionBeingAttempted :type EInteractionBeingAttempted%20Enum :default "FirstInteraction"))
  (bind _returnvalue (Fortnite|GetFortPlayerControllerfromActor InteractingPawn))
  (Development|ExecuteConsoleCommand "Cheat GiveSpecificItem /Script/FortniteGame.AthenaGadgetItemDefinition'/IcicleLarvaGameplay/Gameplay/AGID_IcicleLarva_Special.AGID_IcicleLarva_Special' 1 1" _returnvalue)
  (Utilities|FlowControl|Delay 1.0)
  (Development|ExecuteConsoleCommand "cheat grantquest Quest_PistolCanary_Found_RingTimeAttack_Urgent_Repeat" _returnvalue))

(fn BlueprintCanInteract
  (params (InteractingPawn :type Fort%20Pawn%20Object%20Reference) (InteractionBeingAttempted :type EInteractionBeingAttempted%20Enum :default "FirstInteraction") (InteractionType :type TInteractionType%20Enum :default "IT_NoInteraction"))
  (return true))

(fn BlueprintGetInteractionString
  (params (InteractingPawn :type Fort%20Pawn%20Object%20Reference) (InteractionBeingAttempted :type EInteractionBeingAttempted%20Enum :default "FirstInteraction"))
  (return "NSLOCTEXT(\"[FC7142E80CA68E182EEE69477A2B3600]\", \"ECF311704AB57E277D3ADE92D0B04C71\", \"Time Attack\")"))

(fn ConstructionScript)
