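; BP_PC_SDBGA_Ring is a ring quest-item Blueprint for the Season 36 Battle Pass quests that hands state tree completion to a timer manager.
; When its FortPlayerInstancedStateTree run status changes to Succeeded, it finds the BP_PC_TimeManager by class and calls SetTimer with 0.
; BeginPlay, overlap and tick events are disabled stubs, and the construction script only calls its parent.
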(event EventBeginPlay
  (|CallParentFunction) ; This node is disabled and will not be called.\nDrag off pins to build functionality.
)

(event Collision|EventActorBeginOverlap
  (params (OtherActor :type Actor%20Object%20Reference))
  (|CallParentFunction OtherActor) ; This node is disabled and will not be called.\nDrag off pins to build functionality.
)

(event EventTick
  (params (DeltaSeconds :type Float%20%28single-precision%29))
  (|CallParentFunction DeltaSeconds) ; This node is disabled and will not be called.\nDrag off pins to build functionality.
)

(event OnStateTreeRunStatusChanged(FortPlayerInstancedStateTree)
  (params (StateTreeRunStatus :type EStateTreeRunStatus%20Enum :default "Running"))
  (switch Utilities|FlowControl|Switch|SwitchonEStateTreeRunStatus StateTreeRunStatus
    (:Succeeded
      (bind _returnvalue (Actor|GetActorOfClass "/BattlePassS36_Quests/Items/QuestItems/Rings/BP_PC_TimeManager.BP_PC_TimeManager_C"))
      (Class|BPPCTimeManager|SetTimer _returnvalue 0 :!class /BattlePassS36_Quests/Items/QuestItems/Rings/BP_PC_TimeManager.BP_PC_TimeManager_C))))

(fn ConstructionScript
  (|CallParentFunction))
