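; Time-attack manager that times a player's run through the level's BP_PC_SDBGA_Ring actors.
; BeginPlay stores the ring count as the target.
; Each SetTimer call records the pawn and increments a pass counter, with the first enabling Tick to accumulate and print elapsed time.
; Reaching the target stops Tick, prints the final time, resets, tosses the player's current weapon and ignores their next fall damage.
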(event EventBeginPlay
  (bind _outactors (Actor|GetAllActorsOfClass "/BattlePassS36_Quests/Items/QuestItems/Rings/BP_PC_SDBGA_Ring.BP_PC_SDBGA_Ring_C"))
  (Variables|Default|SetMax (Utilities|Array|Length _outactors)))

(event Custom|SetTimer
  (params (FortPlayerPawn :type Fort%20Player%20Pawn%20Object%20Reference))
  (bind _fort_player_pawn (Variables|Default|GetFort%20Player%20Pawn))
  (bind _returnvalue (Fortnite|GetFortPlayerControllerfromActor _fort_player_pawn))
  (bind _ranthrough (Variables|Default|GetRanThrough))
  (Variables|Default|SetFort%20Player%20Pawn FortPlayerPawn)
  (if (<= _ranthrough 0)
    (Development|PrintString "TIME STARTED")
    (CallFunction|TimeAttack true))
  (Math|Integer|IncrementInt _ranthrough
    (:%20%20
      (if (>= (Variables|Default|GetRanThrough) (Variables|Default|GetMax))
        (CallFunction|TimeAttack false)
        (Development|PrintString (Utilities|String|Append "FINAL TIME
" (Utilities|String|ToString(Float) (Variables|Default|GetTimer))) true true "(R=0.000000,G=0.660000,B=1.000000,A=1.000000)" 1e+01)
        (Variables|Default|SetRanThrough 0)
        (Variables|Default|SetTimer 0.0)
        (bind _returnvalue_2 (Inventory|GetInventoryItemwithGuid _returnvalue (FortWeapon|GetInventoryGUID (Weapon|GetCurrentWeapon _fort_player_pawn))))
        (Inventory|TossSpecificItem _returnvalue (Item|GetItemDefinition _returnvalue_2) (Utilities|Struct|MakeFortPickupTossOverrideData false 3e+02 6e+02 9e+01))
        (Variables|/Script/FortniteGame.FortPawn|SetbIgnoreNextFallingDamage true (Variables|Default|GetFort%20Player%20Pawn))))))

(event Custom|TimeAttack
  (params (bEnabled :type Boolean))
  (Actor|Tick|SetActorTickEnabled bEnabled))

(event Collision|EventActorBeginOverlap
  (params (OtherActor :type Actor%20Object%20Reference)))

(event EventTick
  (params (DeltaSeconds :type Float%20%28single-precision%29))
  (bind _output_get (Variables|Default|SetTimer (+ DeltaSeconds (Variables|Default|GetTimer))))
  (Development|PrintString (Utilities|String|ToString(Float) _output_get)))

(fn ConstructionScript)
